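# Physics asset for the KelpLinenMagnesium axe pickaxe mesh, giving its dangling chain physical bodies and joints.
# A capsule covers the two-handed handle, with further capsules on the chain base and six dynamic chain bones.
# Constraints link the chain bones in series with limited swing and twist, the first joint allowing a wider swing.
# The chain base attaches to the handle through a joint with free swing and twist.
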PhysicsAsset /BRCosmetics/Weapons/FORT_Melee/Pickaxe_KelpLinenMagnesium/Meshes/KelpLinenMagnesium_Axe_Physics
bodies {
 Base_sword_chain [{shapeType=Capsule, length=0.505, radius=0.505, shapeName=Base_sword_chain_capsule, rotation=<rot 0 -90 0>}]
 Melee_TwoHanded_Handle [{
   shapeType=Capsule
   length=37.8437
   radius=2.4724
   shapeName=Melee_TwoHanded_Handle_capsule
   center=<0.0109 -0.5837 10.4046>
   rotation=<rot -1.1241 -90 0.0393>
  }]
 dyn_sword_chain_01 [{
   shapeType=Capsule
   length=0.505
   radius=0.505
   shapeName=dyn_sword_chain_01_capsule
   center=<1.4163 0.0016 0>
   rotation=<rot 0 -90.0143 -86.5583>
  }]
 dyn_sword_chain_02 [{
   shapeType=Capsule
   length=1.821
   radius=1.7
   shapeName=dyn_sword_chain_02_capsule
   center=<1.4239 0.0032 -0.0669>
   rotation=<rot 0 -86.7509 -88.8639>
  }]
 dyn_sword_chain_03 [{
   shapeType=Capsule
   length=0.505
   radius=1.7
   shapeName=dyn_sword_chain_03_capsule
   center=<1.437 0.0016 0>
   rotation=<rot 0 -90.0143 -86.5583>
  }]
 dyn_sword_chain_04 [{
   shapeType=Capsule
   length=1.821
   radius=1.7
   shapeName=dyn_sword_chain_04_capsule
   center=<1.3828 -0.0425 -0.0938>
   rotation=<rot 0 86.7409 -88.8932>
  }]
 dyn_sword_chain_05 [{
   shapeType=Capsule
   length=1.821
   radius=1.7
   shapeName=dyn_sword_chain_05_capsule
   center=<1.4101 0.0015 0>
   rotation=<rot 0 -91.1023 -86.7394>
  }]
 dyn_sword_chain_06 [{
   shapeType=Capsule
   length=1.8543
   radius=1.7
   shapeName=dyn_sword_chain_06_capsule
   center=<1.4418 -0.0345 -0.0674>
   rotation=<rot 0 83.4893 -87.8314>
  }]
}
constraints [
 {bone1Name=dyn_sword_chain_01, bone2Name=Base_sword_chain, swing1LimitDegrees=45, swing2LimitDegrees=30, twistLimitDegrees=25}
 {bone1Name=dyn_sword_chain_02, bone2Name=dyn_sword_chain_01, swing1LimitDegrees=30, swing2LimitDegrees=30, twistLimitDegrees=25}
 {bone1Name=dyn_sword_chain_03, bone2Name=dyn_sword_chain_02, swing1LimitDegrees=30, swing2LimitDegrees=30, twistLimitDegrees=25}
 {bone1Name=dyn_sword_chain_04, bone2Name=dyn_sword_chain_03, swing1LimitDegrees=30, swing2LimitDegrees=30, twistLimitDegrees=25}
 {bone1Name=dyn_sword_chain_05, bone2Name=dyn_sword_chain_04, swing1LimitDegrees=30, swing2LimitDegrees=30, twistLimitDegrees=25}
 {bone1Name=dyn_sword_chain_06, bone2Name=dyn_sword_chain_05, swing1LimitDegrees=30, swing2LimitDegrees=30, twistLimitDegrees=25}
 {
  bone1Name=Base_sword_chain
  bone2Name=Melee_TwoHanded_Handle
  swing1LimitDegrees=45
  swing1Motion=Free
  swing2LimitDegrees=45
  swing2Motion=Free
  twistLimitDegrees=45
  twistMotion=Free
 }
]
